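; Bounty board actor that players interact with to take bounties, destroyed at BeginPlay in modes that disallow it.
; The server drives interaction text through trigger overlaps, ignoring hunters' controllers, stored so the ignore survives death.
; When the local player gets a bounty, it swaps to the empty board, hides the conversation indicator and flickers the light.
; Reset clears ignores and restores visuals; death hides the actor and deactivates its special-actor component.
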(event EventEndPlay
  (params (EndPlayReason :type EEndPlayReason%20Enum :default "Destroyed"))
  ;[0.56 0.4432 1] Log Debug info to help identify cause of Bounty Boards being deleted in Creative Royale Edit/Play mode.
  ;
  ; This is TEMPORARY and will be removed before public release.
  (|CallParentFunction EndPlayReason)
  (if (ScalableFloat|GetValueasBool (Variables|Default|GetCanPostShippingLogs))
    (Fortnite|FortShippingLog self (Utilities|String|Append "EndPlay caused by " (Utilities|String|EnumtoString EndPlayReason)) true)))

(event EventBeginPlay
  (Fortnite|SetCanbeDamaged self (ScalableFloat|GetValueasBool (Variables|Default|GetbCanBeDamagedHotfix)))
  (Variables|Default|SetCachedPointLightIntensity (Variables|/Script/Engine.LightComponentBase|GetIntensity (Variables|Default|GetPointLight)))
  ;[0.281 1 0.4092] Destroy the vending machine if in Creative Royale.
  (Networking|SwitchHasAuthority
    (:Authority
      (if (not (ScalableFloat|GetValueasBool (Variables|Default|GetAllowedInMode)))
        (Actor|DestroyActor)))))

(event EventOnBountyStarted
  (params (HunterPlayerStates :type Array%20of%20Fort%20Player%20State%20Athena%20Object%20References))
  ;[0 0 0] OnBountyStarted
  (if (Networking|IsServer)
    (for _array_element HunterPlayerStates
      (Utilities|IsValid (Fortnite|GetFortPlayerControllerfromActor _array_element)
        (:Is%20Valid
          ;;[1 1 1] Store *controller* not pawn so we can retain ignores between death
          (Utilities|Array|Add (Variables|Default|GetPlayersToIgnore) (Fortnite|GetFortPlayerControllerfromActor _array_element))
          (Utilities|IsValid (Game|Player|GetPlayerPawn (Fortnite|GetFortPlayerControllerfromActor _array_element))
            (:Is%20Valid
              ;;[0 0 0] This will only succeed on the interacting pawn, will early out for squadmates
              (CallFunction|StopUpdatingInteractionText (Game|Player|GetPlayerPawn (Fortnite|GetFortPlayerControllerfromActor _array_element))))))))))

(event EventOnLocalPlayerGrantedBounty
  (params (HunterPlayerState :type Fort%20Player%20State%20Athena%20Object%20Reference))
  ;[0 0 0] Event On Local Player Granted Bounty
  (Rendering|SetVisibility (Variables|Default|GetBountyBoard_Full) false false)
  (Rendering|SetVisibility (Variables|Default|GetBountyBoard_Empty) true false)
  (Collision|SetCollisionEnabled (Variables|Default|GetBountyBoard_Full) "NoCollision")
  (Collision|SetCollisionEnabled (Variables|Default|GetBountyBoard_Empty) "QueryAndPhysics")
  (Collision|SetCollisionEnabled (Variables|Default|GetTriggerVolume) "NoCollision")
  ;[1 1 1] The NonPlayerConversationComponent will hide itself as long as its list of Reasons is > 0, this is checked in code.
  (InteractionUI|AddHideUIIndicatorReason (Variables|Default|GetNonPlayerConversationComponent) (Variables|Default|GetHideReason))
  (bind (_direction _newtrack_0) (Timeline|LightFlicker PlayFromStart)
    (:Update
      (Rendering|Components|Light|SetIntensity (Variables|Default|GetPointLight) (Math|Float|Lerp 0.0 5.0 _newtrack_0)))))

(event OnComponentBeginOverlap(TriggerVolume)
  (params (OverlappedComponent :type Primitive%20Component%20Object%20Reference) (OtherActor :type Actor%20Object%20Reference) (OtherComp :type Primitive%20Component%20Object%20Reference) (OtherBodyIndex :type Integer) (bFromSweep :type Boolean) (SweepResult :type Hit%20Result%20Structure%20%28by%20ref%29))
  (bind _returnvalue (Fortnite|GetFortPlayerControllerfromActor OtherActor))
  (bind _returnvalue_2 (Game|Player|GetPlayerPawn _returnvalue))
  ;[0 0 0] OnComponentBeginOverlap
  (if (Networking|IsServer)
    (Utilities|IsValid _returnvalue
      (:Is%20Valid
        (Utilities|IsValid _returnvalue_2
          (:Is%20Valid
            (if (not (Utilities|Array|ContainsItem (Variables|Default|GetPlayersToIgnore) _returnvalue))
              ;[0 0 0 0] Even though this actor was moved over to use the NonPlayerConversationComponent, believe this is still needed to set some things natively for bounty logic. Don't remove.
              (CallFunction|StartUpdatingInteractionText _returnvalue_2))))))))

(event OnComponentEndOverlap(TriggerVolume)
  (params (OverlappedComponent :type Primitive%20Component%20Object%20Reference) (OtherActor :type Actor%20Object%20Reference) (OtherComp :type Primitive%20Component%20Object%20Reference) (OtherBodyIndex :type Integer))
  (bind _returnvalue (Fortnite|GetFortPlayerControllerfromActor OtherActor))
  (bind _returnvalue_2 (Game|Player|GetPlayerPawn _returnvalue))
  ;[0 0 0] OnComponentEndOverlap
  (if (Networking|IsServer)
    (Utilities|IsValid _returnvalue
      (:Is%20Valid
        (Utilities|IsValid _returnvalue_2
          (:Is%20Valid
            (if (not (Utilities|Array|ContainsItem (Variables|Default|GetPlayersToIgnore) _returnvalue))
              (CallFunction|StopUpdatingInteractionText _returnvalue_2))))))))

(event EventOnDeathServer
  (params (Damage :type Float%20%28single-precision%29) (DamageTags :type Gameplay%20Tag%20Container%20Structure%20%28by%20ref%29) (Momentum :type Vector :default "0, 0, 0") (HitInfo :type Hit%20Result%20Structure%20%28by%20ref%29) (InstigatedBy :type Controller%20Object%20Reference) (DamageCauser :type Actor%20Object%20Reference) (EffectContext :type Gameplay%20Effect%20Context%20Handle%20Structure))
  (bind _bountyboardspecialactorcomponent (Variables|Default|GetBountyBoardSpecialActorComponent))
  (Rendering|SetActorHiddenInGame true)
  (Utilities|IsValid _bountyboardspecialactorcomponent
    (:Is%20Valid
      (Components|Activation|Deactivate _bountyboardspecialactorcomponent))))

(event EventOnBountyBoardReset
  ;[0.5094 0.9701 1] Reset the bounty board to its initial state
  (if (Networking|IsServer)
    (Utilities|Array|Clear (Variables|Default|GetPlayersToIgnore))
    (else
      ;;[1 0.5571 0.8611] Reset cosmetic changes
      (Networking|CanExecuteCosmeticEvents
        (:True
          (Rendering|SetVisibility (Variables|Default|GetBountyBoard_Full) true false)
          (Rendering|SetVisibility (Variables|Default|GetBountyBoard_Empty) false false)
          (Collision|SetCollisionEnabled (Variables|Default|GetBountyBoard_Full) "QueryOnly")
          (Collision|SetCollisionEnabled (Variables|Default|GetBountyBoard_Empty) "NoCollision")
          (Collision|SetCollisionEnabled (Variables|Default|GetTriggerVolume) "QueryOnly")
          (InteractionUI|RemoveHideUIIndicatorReason (Variables|Default|GetNonPlayerConversationComponent) (Variables|Default|GetHideReason))
          (Rendering|Components|Light|SetIntensity (Variables|Default|GetPointLight) (Variables|Default|GetCachedPointLightIntensity)))))))

(fn BlueprintGetFailedInteractionString
  (params (InteractingPawn :type Fort%20Pawn%20Object%20Reference) (InteractionBeingAttempted :type EInteractionBeingAttempted%20Enum :default "FirstInteraction"))
  (return (Variables|GetInteractionText)))

(fn BlueprintGetInteractionString
  (params (InteractingPawn :type Fort%20Pawn%20Object%20Reference) (InteractionBeingAttempted :type EInteractionBeingAttempted%20Enum :default "FirstInteraction"))
  (return (Variables|GetInteractionText)))

(fn BlueprintGetInteractionTime
  (params (InteractingPawn :type Fort%20Pawn%20Object%20Reference) (InteractionBeingAttempted :type EInteractionBeingAttempted%20Enum :default "FirstInteraction"))
  (return 0.0 true))

(fn ConstructionScript
  (|CallParentFunction))
