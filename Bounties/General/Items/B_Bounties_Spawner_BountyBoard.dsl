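; Bounty board spawner Blueprint for the Bounties feature, with spawning behavior not defined in this graph.
; BuildDataRegistryResolverScope finds the owning minigame and returns its spatial gameplay FName as a data registry scope, which GetValueWithContext lookups require.
; GetCacheContextOverride returns zero.
; BeginPlay, overlap and tick events are empty or disabled stubs, and construction only calls the parent.
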(event EventBeginPlay
  (|CallParentFunction) ; This node is disabled and will not be called.\nDrag off pins to build functionality.
)

(event Collision|EventActorBeginOverlap
  (params (OtherActor :type Actor%20Object%20Reference)))

(event EventTick
  (params (DeltaSeconds :type Float%20%28single-precision%29)))

(fn BuildDataRegistryResolverScope
  ;[1 0.8475 0.2776] Gives the name of the volume as a scope to the data registry. Needed for calls to GetValueWithContext to work
  (return (Utilities|Array|MakeArray (SpatialGameplay|GetSpatialGameplayInterfaceFNameforActor (Minigame|GetMinigame self :!class /Script/FortniteGame.MinigameBlueprintLibrary))) 0 true))

(fn GetCacheContextOverride
  (return 0))

(fn ConstructionScript
  (|CallParentFunction))
